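; Gameplay ability for a Season 20 battle pass quest about entering a vehicle after landing from the bus.
; When activated by an event, it calls the parent, then waits for the Event.SkydiveEnded.FromBus tag to be removed.
; On removal it applies GE_Quest_Enter_After_Landing_10s to its owner, apparently a ten-second quest window, and ends.
; The vehicle-entry check itself is not in this graph.
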(event Ability|EventActivateAbilityFromEvent
  (params (EventData :type Gameplay%20Event%20Data%20Structure%20%28by%20ref%29))
  (|CallParentFunction EventData)
  (Ability|Tasks|WaitGameplayTagRemove "(TagName=\"Event.SkydiveEnded.FromBus\")" 0 true
    (:Removed
      (Ability|ApplyGameplayEffectToOwner "/BattlePassS20/Items/QuestAbilities/Quest_EnterVehicle_After_Landing/GE_Quest_Enter_After_Landing_10s.GE_Quest_Enter_After_Landing_10s_C")
      (Ability|EndAbility))))
